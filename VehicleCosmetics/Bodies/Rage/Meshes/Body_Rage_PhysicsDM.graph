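# Physics asset for the Rage vehicle body cosmetic mesh, in a PhysicsDM variant.
# It holds a single root body with one box shape offset forward and up from the origin to enclose the car body.
# It has no constraints, so the body collides and simulates as one rigid piece.
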PhysicsAsset /VehicleCosmetics/Bodies/Rage/Meshes/Body_Rage_PhysicsDM
bodies {root [{shapeType=Box, extentX=444, extentY=216, extentZ=116, shapeName=root_box, center=<22.4361 0 93.9186>}]}
constraints []
